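# Physics asset for the medium male Dummeez player body, set up for a dangling scarf, likely as secondary motion.
# Two large capsules on upper spine bones form the torso volume, and a tiny driver capsule links the scarf to spine_04.
# Three scarf capsules shrink along a chain from the driver, with tight twist limits and one swing limit widening toward the tip.
# Spine and driver joints leave swing and twist free.
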PhysicsAsset /BRCosmetics/Characters/Player/Male/Medium/Bodies/M_MED_Dummeez/Meshes/M_MED_Dummeez_Physics
bodies {
 dyn_driver_dyn_scarf [{shapeType=Capsule, length=0.505, radius=0.505, rotation=<rot 0 -90 0>}]
 dyn_scarf_01 [{shapeType=Capsule, length=5.5, radius=4.75, center=<1 -0.0001 4.5>, rotation=<rot 90 0 -90>}]
 dyn_scarf_02 [{shapeType=Capsule, length=5.25, radius=4, center=<2 0 3.5>, rotation=<rot 90 90 0>}]
 dyn_scarf_03 [{shapeType=Capsule, length=3.75, radius=3.25, center=<3 0 1.5>, rotation=<rot 90 90 0>}]
 spine_04 [{shapeType=Capsule, length=10.5, radius=15.5, center=<3.5034 -4.7546 0>}]
 spine_05 [{shapeType=Capsule, length=12.5, radius=14.75, center=<0.0679 -3.4524 0>}]
}
constraints [
 {
  bone1Name=spine_05
  bone2Name=spine_04
  swing1LimitDegrees=45
  swing1Motion=Free
  swing2LimitDegrees=45
  swing2Motion=Free
  twistLimitDegrees=45
  twistMotion=Free
 }
 {
  bone1Name=dyn_driver_dyn_scarf
  bone2Name=spine_04
  swing1LimitDegrees=45
  swing1Motion=Free
  swing2LimitDegrees=45
  swing2Motion=Free
  twistLimitDegrees=45
  twistMotion=Free
 }
 {bone1Name=dyn_scarf_01, bone2Name=dyn_driver_dyn_scarf, swing1LimitDegrees=20, swing2LimitDegrees=20, twistLimitDegrees=5}
 {bone1Name=dyn_scarf_02, bone2Name=dyn_scarf_01, swing1LimitDegrees=20, swing2LimitDegrees=35, twistLimitDegrees=5}
 {bone1Name=dyn_scarf_03, bone2Name=dyn_scarf_02, swing1LimitDegrees=20, swing2LimitDegrees=45, twistLimitDegrees=5}
]
