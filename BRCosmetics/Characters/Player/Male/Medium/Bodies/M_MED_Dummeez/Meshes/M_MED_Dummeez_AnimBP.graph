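# Animation blueprint for the Dummeez male medium body cosmetic, adding secondary physics motion on top of the parent character's pose.
# It copies the pose from the attached parent mesh, then runs a RigidBody node with the body's physics asset in base-bone space, using damped component motion and custom gravity.
# While skydiving, a quickly blended additive offset nudges and rolls the dyn_backpack bone before the final output.
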AnimBlueprint /BRCosmetics/Characters/Player/Male/Medium/Bodies/M_MED_Dummeez/Meshes/M_MED_Dummeez_AnimBP
target_skeleton=&/M_MED_Dummeez_Skeleton
graphs [
 AnimGraph {nodes [
   n1 = RigidBody(ComponentPose=Local To Component(LocalPose=Copy Pose From Mesh() {bUseAttachedParent=true})) {
    bClampLinearTranslationLimitToRefPose=true
    bOverrideWorldGravity=true
    overridePhysicsAsset=&/M_MED_Dummeez_Physics
    simulationSpace=BaseBoneSpace
    alphaBoolBlend {blendOption=Linear}
    baseBoneRef {boneName=root}
    componentAppliedLinearAccClamp=<1100 1100 1100>
    componentLinearAccScale=<0.6 0.6 0.6>
    componentLinearVelScale=<0.6 0.6 0.6>
    overrideWorldGravity {z=-980}
   }
   n2 = "Transform (Modify) Bone - Bone: dyn_backpack"!(ComponentPose=@n1, bAlphaBoolEnabled:bool=Get bIsSkyDiving(), Translation="0.000,-3,0", Rotation="0,0,-5", Scale="0,0,0") -> (Pose) {
    alphaInputType=Bool
    rotationMode=BMM_Additive
    rotationSpace=BCS_BoneSpace
    translationMode=BMM_Additive
    translationSpace=BCS_BoneSpace
    alphaBoolBlend {blendInTime=0.1, blendOption=Linear, blendOutTime=0.1}
    boneToModify {boneName=dyn_backpack}
    rotation {roll=-5}
    scale=<0 0 0>
    translation {y=-3}
   }
   Output Pose(Result=Component To Local(ComponentPose=@n2)) {layerGroup=None}
  ]}
 EventGraph {links [], nodes []}
]
variables []
